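# Post-copy AnimBlueprint for the Sandcastle Tattoos male player body, adding procedural muscle deformation and chain physics to the parent mesh's pose.
# Constraints and look-ats drive pecs, scapulae and lats, while shoulder-to-forearm distances scale additive bicep and tricep bulges.
# Constrained chain drivers then feed a RigidBody whose gravity follows skydiving force.
# Disabled rigid bodies or LOD above 1 bypass everything; bracelet and shoe physics nodes are unconnected.
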AnimBlueprint /Game/Characters/Player/Male/Medium/Bodies/M_MED_Sandcastle_Tattoos/Meshes/M_MED_Sandcastle_Tattoos_AnimBP
target_skeleton=&/M_MED_Sandcastle_Tattoos_Skeleton
graphs [
 AnimGraph {nodes [
   n1 = Get LarmDistanceBICEPAlpha() -> (LarmDistanceBICEPAlpha:double)
   n2 = "Transform (Modify) Bone - Bone: bicep_02_l"(Alpha=@n1, Translation="-6,-0.000,-0.000", Rotation="0,0,0", Scale="1,1,1") {
    rotationMode=BMM_Additive
    rotationSpace=BCS_BoneSpace
    scaleMode=BMM_Additive
    scaleSpace=BCS_BoneSpace
    translationMode=BMM_Additive
    translationSpace=BCS_BoneSpace
    alphaBoolBlend {blendOption=Linear}
    boneToModify {boneName=bicep_02_l}
    translation {x=-6}
   }
   n3 = "Transform (Modify) Bone - Bone: bicep_01_l"(ComponentPose=@n2, Alpha=@n1, Translation="-0.000,0.000,2.343", Rotation="0,0,0", Scale="1.000,1.500,1.250") {
    rotationMode=BMM_Additive
    rotationSpace=BCS_BoneSpace
    scaleMode=BMM_Additive
    scaleSpace=BCS_BoneSpace
    translationMode=BMM_Additive
    translationSpace=BCS_BoneSpace
    alphaBoolBlend {blendOption=Linear}
    boneToModify {boneName=bicep_01_l}
    scale {y=1.5, z=1.25}
    translation {z=2.343}
   }
   n4 = "Transform (Modify) Bone - Bone: tricep_02_l"(ComponentPose=@n3, Alpha=@n1, Translation="-6,-0.000,-0.000", Rotation="0,0,0", Scale="1,1,1") {
    rotationMode=BMM_Additive
    rotationSpace=BCS_BoneSpace
    scaleMode=BMM_Additive
    scaleSpace=BCS_BoneSpace
    translationMode=BMM_Additive
    translationSpace=BCS_BoneSpace
    alphaBoolBlend {blendOption=Linear}
    boneToModify {boneName=tricep_02_l}
    translation {x=-6}
   }
   n5 = "Transform (Modify) Bone - Bone: tricep_01_l"(ComponentPose=@n4, Alpha=@n1, Translation="-0.000,0.000,2.343", Rotation="0,0,0", Scale="1.000,1.500,1.250") {
    rotationMode=BMM_Additive
    rotationSpace=BCS_BoneSpace
    scaleMode=BMM_Additive
    scaleSpace=BCS_BoneSpace
    translationMode=BMM_Additive
    translationSpace=BCS_BoneSpace
    alphaBoolBlend {blendOption=Linear}
    boneToModify {boneName=tricep_01_l}
    scale {y=1.5, z=1.25}
    translation {z=2.343}
   }
   n6 = Get RarmDistanceBICEPAlpha() -> (RarmDistanceBICEPAlpha:double)
   n7 = "Transform (Modify) Bone - Bone: bicep_02_r"(ComponentPose=@n5, Alpha=@n6, Translation="-6,-0.000,-0.000", Rotation="0,0,0", Scale="1,1,1") {
    rotationMode=BMM_Additive
    rotationSpace=BCS_BoneSpace
    scaleMode=BMM_Additive
    scaleSpace=BCS_BoneSpace
    translationMode=BMM_Additive
    translationSpace=BCS_BoneSpace
    alphaBoolBlend {blendOption=Linear}
    boneToModify {boneName=bicep_02_r}
    translation {x=-6}
   }
   n8 = "Transform (Modify) Bone - Bone: bicep_01_r"(ComponentPose=@n7, Alpha=@n6, Translation="-0.000,0.000,2.343", Rotation="0,0,0", Scale="1.000,1.500,1.250") {
    rotationMode=BMM_Additive
    rotationSpace=BCS_BoneSpace
    scaleMode=BMM_Additive
    scaleSpace=BCS_BoneSpace
    translationMode=BMM_Additive
    translationSpace=BCS_BoneSpace
    alphaBoolBlend {blendOption=Linear}
    boneToModify {boneName=bicep_01_r}
    scale {y=1.5, z=1.25}
    translation {z=2.343}
   }
   n9 = "Transform (Modify) Bone - Bone: tricep_02_r"(ComponentPose=@n8, Alpha=@n6, Translation="-6,-0.000,-0.000", Rotation="0,0,0", Scale="1,1,1") {
    rotationMode=BMM_Additive
    rotationSpace=BCS_BoneSpace
    scaleMode=BMM_Additive
    scaleSpace=BCS_BoneSpace
    translationMode=BMM_Additive
    translationSpace=BCS_BoneSpace
    alphaBoolBlend {blendOption=Linear}
    alphaScaleBiasClamp {scale=0.5}
    boneToModify {boneName=tricep_02_r}
    translation {x=-6}
   }
   n10 = "Transform (Modify) Bone - Bone: tricep_01_r"(ComponentPose=@n9, Alpha=@n6, Translation="-0.000,0.000,2.343", Rotation="0,0,0", Scale="1.000,1.500,1.250") {
    rotationMode=BMM_Additive
    rotationSpace=BCS_BoneSpace
    scaleMode=BMM_Additive
    scaleSpace=BCS_BoneSpace
    translationMode=BMM_Additive
    translationSpace=BCS_BoneSpace
    alphaBoolBlend {blendOption=Linear}
    alphaScaleBiasClamp {scale=0.5}
    boneToModify {boneName=tricep_01_r}
    scale {y=1.5, z=1.25}
    translation {z=2.343}
   }
   Component To Local(ComponentPose=@n10)
   n11 = "Constraint - scapula_base_l (P)"(ComponentPose=Local To Component(), Alpha=1, ConstraintWeights_0=1) {
    alphaBoolBlend {blendOption=Linear}
    boneToModify {boneName=scapula_base_l}
    constraintSetup [{targetBone {boneName=clavicle_l}}]
    constraintWeights [1]
   }
   n12 = Look At - Bone: scapula_00_l(ComponentPose=@n11, Alpha=1) {
    alphaBoolBlend {blendOption=Linear}
    boneToModify {boneName=scapula_00_l}
    lookAtLocation {x=0}
    lookAtTarget {boneReference {boneName=scapula_01_l_aim}}
    lookAt_Axis {axis=<1 0>}
    lookUp_Axis {axis {x=0, z=1}}
   }
   n13 = Look At - Bone: pectoralis_01_l(ComponentPose=@n12, Alpha=1) {
    alphaBoolBlend {blendOption=Linear}
    boneToModify {boneName=pectoralis_01_l}
    lookAtLocation {x=0}
    lookAtTarget {boneReference {boneName=upperarm_l}}
    lookAt_Axis {axis=<1 0>}
    lookUp_Axis {axis {x=0, z=1}}
   }
   n14 = Look At - Bone: lattissimus_00_l(ComponentPose=@n13, Alpha=1) {
    alphaBoolBlend {blendOption=Linear}
    boneToModify {boneName=lattissimus_00_l}
    lookAtLocation {x=0}
    lookAtTarget {boneReference {boneName=upperarm_l}}
    lookAt_Axis {axis=<1 0>}
    lookUp_Axis {axis {x=0, z=1}}
   }
   n15 = "Constraint - scapula_base_r (P)"(ComponentPose=@n14, Alpha=1, ConstraintWeights_0=1) {
    alphaBoolBlend {blendOption=Linear}
    boneToModify {boneName=scapula_base_r}
    constraintSetup [{targetBone {boneName=clavicle_r}}]
    constraintWeights [1]
   }
   n16 = Look At - Bone: scapula_00_r(ComponentPose=@n15, Alpha=1) {
    alphaBoolBlend {blendOption=Linear}
    boneToModify {boneName=scapula_00_r}
    lookAtLocation {x=0}
    lookAtTarget {boneReference {boneName=scapula_01_r_aim}}
    lookAt_Axis {axis=<1 0>}
    lookUp_Axis {axis {x=0, z=1}}
   }
   n17 = Look At - Bone: pectoralis_01_r(ComponentPose=@n16, Alpha=1) {
    alphaBoolBlend {blendOption=Linear}
    boneToModify {boneName=pectoralis_01_r}
    lookAtLocation {x=0}
    lookAtTarget {boneReference {boneName=upperarm_r}}
    lookAt_Axis {axis=<1 0>}
    lookUp_Axis {axis {x=0, z=1}}
   }
   Look At - Bone: lattissimus_00_r(ComponentPose=@n17, Alpha=1) {
    alphaBoolBlend {blendOption=Linear}
    boneToModify {boneName=lattissimus_00_r}
    lookAtLocation {x=0}
    lookAtTarget {boneReference {boneName=upperarm_r}}
    lookAt_Axis {axis=<1 0>}
    lookUp_Axis {axis {x=0, z=1}}
   }
   n18 = RigidBody!(ComponentPose, Alpha, OverrideWorldGravity:Vector="0,0,-980", ExternalForce) -> (Pose) {
    bClampLinearTranslationLimitToRefPose=true
    bOverrideWorldGravity=true
    overridePhysicsAsset=&/M_MED_Sandcastle_Tattoos_Bracelet_Physics
    simulationSpace=BaseBoneSpace
    alphaBoolBlend {blendOption=Linear}
    baseBoneRef {boneName=root}
    componentAppliedLinearAccClamp=<3000 3000 3000>
    componentLinearAccScale=<0.5 0.5 0.5>
    componentLinearVelScale=<0.5 0.5 0.5>
    overrideWorldGravity {z=-980}
   }
   RigidBody(ComponentPose=@n18, OverrideWorldGravity="0,0,-980") {
    bClampLinearTranslationLimitToRefPose=true
    bOverrideWorldGravity=true
    overridePhysicsAsset=&/M_MED_Sandcastle_Tattoos_Shoes_Physics
    simulationSpace=BaseBoneSpace
    alphaBoolBlend {blendOption=Linear}
    baseBoneRef {boneName=root}
    componentAppliedLinearAccClamp=<3000 3000 3000>
    componentLinearAccScale=<0.5 0.5 0.5>
    componentLinearVelScale=<0.5 0.5 0.5>
    overrideWorldGravity {z=-980}
   }
   n19 = "Constraint - dyn_chaindriver_bk_01_r_MSTR (PPPPPPP)"!(
    ComponentPose=Local To Component(LocalPose=Use cached pose 'anatomyFinal'())
    Alpha=1
    ConstraintWeights_0=0
    ConstraintWeights_1:float=0.2059
    ConstraintWeights_2:float=0.0214
    ConstraintWeights_3:float=0
    ConstraintWeights_4:float=0.0001
    ConstraintWeights_5:float=0.7721
    ConstraintWeights_6:float=0.0006
   ) -> (Pose) {
    alphaBoolBlend {blendOption=Linear}
    boneToModify {boneName=dyn_chaindriver_bk_01_r_MSTR}
    constraintSetup [
     {targetBone {boneName=thigh_r}}
     {targetBone {boneName=thigh_twist_01_r}}
     {targetBone {boneName=spine_01}}
     {targetBone {boneName=spine_02}}
     {targetBone {boneName=spine_03}}
     {targetBone {boneName=pelvis}}
     {targetBone {boneName=thigh_twist_01_l}}
    ]
    constraintWeights [1,1,1,1,1,1,1]
   }
   n20 = "Constraint - dyn_chaindriver_md_01_r_MSTR (R)"(ComponentPose=@n19, Alpha=1, ConstraintWeights_0=1) {
    alphaBoolBlend {blendOption=Linear}
    boneToModify {boneName=dyn_chaindriver_md_01_r_MSTR}
    constraintSetup [{transformType=Rotation, targetBone {boneName=thigh_twist_01_r}}]
    constraintWeights [1]
   }
   n21 = "Constraint - dyn_chaindriver_fr_01_r_MSTR (PPPPPPP)"!(
    ComponentPose=@n20
    Alpha=1
    ConstraintWeights_0=0.016
    ConstraintWeights_1:float=0.3068
    ConstraintWeights_2:float=0.0172
    ConstraintWeights_3:float=0.0056
    ConstraintWeights_4:float=0
    ConstraintWeights_5:float=0.6543
    ConstraintWeights_6:float=0.0001
   ) -> (Pose) {
    alphaBoolBlend {blendOption=Linear}
    boneToModify {boneName=dyn_chaindriver_fr_01_r_MSTR}
    constraintSetup [
     {targetBone {boneName=thigh_twist_01_r}}
     {targetBone {boneName=spine_01}}
     {targetBone {boneName=spine_02}}
     {targetBone {boneName=spine_03}}
     {targetBone {boneName=spine_04}}
     {targetBone {boneName=pelvis}}
     {targetBone {boneName=thigh_twist_01_l}}
    ]
    constraintWeights [1,1,1,1,1,1,1]
   }
   n22 = "Constraint - dyn_chaindriver_fr_01_r_MSTR (R)"(ComponentPose=@n21, Alpha=1, ConstraintWeights_0=1) {
    alphaBoolBlend {blendOption=Linear}
    boneToModify {boneName=dyn_chaindriver_fr_01_r_MSTR}
    constraintSetup [{transformType=Rotation, targetBone {boneName=thigh_r}}]
    constraintWeights [1]
   }
   n23 = "Constraint - dyn_chaindriver_bk_01_r_MSTR (R)"(ComponentPose=@n22, Alpha=0, ConstraintWeights_0=1) {
    alphaBoolBlend {blendOption=Linear}
    boneToModify {boneName=dyn_chaindriver_bk_01_r_MSTR}
    constraintSetup [{transformType=Rotation, targetBone {boneName=thigh_twist_01_r}}]
    constraintWeights [1]
   }
   n24 = Look At - Bone: dyn_chaindriver_fr_01_r_MSTR(ComponentPose=@n23, Alpha=0) {
    alphaBoolBlend {blendOption=Linear}
    boneToModify {boneName=dyn_chaindriver_fr_01_r_MSTR}
    lookAtLocation {x=0}
    lookAtTarget {boneReference {boneName=dyn_chaindriver_fr_05_r_aim}}
    lookAt_Axis {axis=<1 0>}
    lookUp_Axis {axis {x=0, z=1}}
   }
   n25 = Look At - Bone: dyn_chaindriver_md_01_r_MSTR(ComponentPose=@n24, Alpha=1) {
    alphaBoolBlend {blendOption=Linear}
    boneToModify {boneName=dyn_chaindriver_md_01_r_MSTR}
    lookAtLocation {x=0}
    lookAtTarget {boneReference {boneName=dyn_chaindriver_md_05_r_aim}}
    lookAt_Axis {axis=<1 0>}
    lookUp_Axis {axis {x=0, z=1}}
   }
   n26 = Look At - Bone: dyn_chaindriver_bk_01_r_MSTR(ComponentPose=@n25, Alpha=1) {
    alphaBoolBlend {blendOption=Linear}
    boneToModify {boneName=dyn_chaindriver_bk_01_r_MSTR}
    lookAtLocation {x=0}
    lookAtTarget {boneReference {boneName=dyn_chaindriver_bk_05_r_aim}}
    lookAt_Axis {axis=<1 0>}
    lookUp_Axis {axis {x=0, z=1}}
   }
   n27 = RigidBody(ComponentPose=@n26, OverrideWorldGravity=Get windVector() -> (windVector:Vector)) {
    bClampLinearTranslationLimitToRefPose=true
    bOverrideWorldGravity=true
    overridePhysicsAsset=&/M_MED_Sandcastle_Tattoos_Chain_Physics
    simulationSpace=BaseBoneSpace
    alphaBoolBlend {blendOption=Linear}
    baseBoneRef {boneName=root}
    componentAppliedLinearAccClamp=<3000 3000 3000>
    componentLinearAccScale=<0.3 0.3 0.3>
    componentLinearVelScale=<0.3 0.3 0.3>
    overrideWorldGravity {z=-980}
   }
   Output Pose(Result=Blend Poses by bool(
     BlendPose_0=Use cached pose 'initCache'()
     BlendPose_1=Component To Local(ComponentPose=@n27)
     bActiveValue=Get rigids_off() -> (rigids_off:bool)
    ) {blendType=Linear}) {layerGroup=None}
   Save cached pose 'initCache'(Pose=Copy Pose From Mesh() {bUseAttachedParent=true}) {cacheName=initCache}
   n28 = Get RarmDistanceBICEPAlpha()
   n29 = Get LarmDistanceBICEPAlpha()
   n30 = "Constraint - pectoralis_02_r_aim (PPPPPPPPPPPPPPPPPP)"!(
    ComponentPose=Local To Component(LocalPose=Use cached pose 'initCache'())
    Alpha=1
    ConstraintWeights_0=0
    ConstraintWeights_1:float=0.0416
    ConstraintWeights_2:float=0.0394
    ConstraintWeights_3:float=0.0003
    ConstraintWeights_4:float=0
    ConstraintWeights_5:float=0
    ConstraintWeights_6:float=0.0122
    ConstraintWeights_7:float=0.259
    ConstraintWeights_8:float=0.0183
    ConstraintWeights_9:float=0.0001
    ConstraintWeights_10:float=0.0106
    ConstraintWeights_11:float=0
    ConstraintWeights_12:float=0.1954
    ConstraintWeights_13:float=0
    ConstraintWeights_14:float=0
    ConstraintWeights_15:float=0
    ConstraintWeights_16:float=0.4228
    ConstraintWeights_17:float=0.0003
   ) -> (Pose) {
    alphaBoolBlend {blendOption=Linear}
    boneToModify {boneName=pectoralis_02_r_aim}
    constraintSetup [
     {targetBone {boneName=tricep_01_r}}
     {targetBone {boneName=upperarm_r}}
     {targetBone {boneName=upperarm_twist_01_r}}
     {targetBone {boneName=deltoid_bk_01_r}}
     {targetBone {boneName=lattissimus_01_r}}
     {targetBone {boneName=pectoralis_02_r}}
     {targetBone {boneName=upperarm_twist_02_r}}
     {targetBone {boneName=deltoid_01_r}}
     {targetBone {boneName=spine_03}}
     {targetBone {boneName=spine_04}}
     {targetBone {boneName=spine_05}}
     {targetBone {boneName=tricep_02_r}}
     {targetBone {boneName=clavicle_r}}
     {targetBone {boneName=neck_02}}
     {targetBone {boneName=neck_01}}
     {targetBone {boneName=bicep_02_r}}
     {targetBone {boneName=deltoid_fr_01_r}}
     {targetBone {boneName=bicep_01_r}}
    ]
    constraintWeights [1,1,1,1,1,1,1,1,1,1,1,1,1,1,1,1,1,1]
   }
   n31 = "Constraint - pectoralis_02_l_aim (PPPPPPPPPPPPPPPPPPPPPPPPPP)"!(
    ComponentPose=@n30
    Alpha=1
    ConstraintWeights_0=0.0003
    ConstraintWeights_1:float=0.0122
    ConstraintWeights_2:float=0
    ConstraintWeights_3:float=0.0183
    ConstraintWeights_4:float=0.0001
    ConstraintWeights_5:float=0.0106
    ConstraintWeights_6:float=0
    ConstraintWeights_7:float=0.4228
    ConstraintWeights_8:float=0.1954
    ConstraintWeights_9:float=0.0394
    ConstraintWeights_10:float=0
    ConstraintWeights_11:float=0
    ConstraintWeights_12:float=0
    ConstraintWeights_13:float=0.0416
    ConstraintWeights_14:float=0
    ConstraintWeights_15:float=0
    ConstraintWeights_16:float=0.259
    ConstraintWeights_17:float=0
    ConstraintWeights_18:float=0
    ConstraintWeights_19:float=0
    ConstraintWeights_20:float=0
    ConstraintWeights_21:float=0.0003
    ConstraintWeights_22:float=0
    ConstraintWeights_23:float=0
    ConstraintWeights_24:float=0
    ConstraintWeights_25:float=0
   ) -> (Pose) {
    alphaBoolBlend {blendOption=Linear}
    boneToModify {boneName=pectoralis_02_l_aim}
    constraintSetup [
     {targetBone {boneName=deltoid_bk_01_l}}
     {targetBone {boneName=upperarm_twist_02_l}}
     {targetBone {boneName=spine_02}}
     {targetBone {boneName=spine_03}}
     {targetBone {boneName=spine_04}}
     {targetBone {boneName=spine_05}}
     {targetBone {boneName=pelvis}}
     {targetBone {boneName=deltoid_fr_01_l}}
     {targetBone {boneName=clavicle_l}}
     {targetBone {boneName=upperarm_twist_01_l}}
     {targetBone {boneName=clavicle_r}}
     {targetBone {boneName=neck_02}}
     {targetBone {boneName=neck_01}}
     {targetBone {boneName=upperarm_l}}
     {targetBone {boneName=lowerarm_twist_02_l}}
     {targetBone {boneName=head}}
     {targetBone {boneName=deltoid_01_l}}
     {targetBone {boneName=lattissimus_01_l}}
     {targetBone {boneName=bicep_02_l}}
     {targetBone {boneName=spine_01}}
     {targetBone {boneName=tricep_02_l}}
     {targetBone {boneName=bicep_01_l}}
     {targetBone {boneName=scapula_01_l}}
     {targetBone {boneName=pectoralis_02_l}}
     {targetBone {boneName=lowerarm_l}}
     {targetBone {boneName=tricep_01_l}}
    ]
    constraintWeights [1,1,1,1,1,1,1,1,1,1,1,1,1,1,1,1,1,1,1,1,1,1,1,1,1,1]
   }
   n32 = Look At - Bone: pectoralis_01_l!(ComponentPose=@n31, Alpha=1, LookAtLocation:Vector="0,0,0") -> (Pose) {
    alphaBoolBlend {blendOption=Linear}
    boneToModify {boneName=pectoralis_01_l}
    lookAtLocation {x=0}
    lookAtTarget {boneReference {boneName=pectoralis_02_l_aim}}
    lookAt_Axis {axis=<1 0>}
    lookUp_Axis {axis {x=0, z=1}}
   }
   n33 = Look At - Bone: pectoralis_01_r!(ComponentPose=@n32, Alpha=1, LookAtLocation:Vector="0,0,0") -> (Pose) {
    alphaBoolBlend {blendOption=Linear}
    boneToModify {boneName=pectoralis_01_r}
    lookAtLocation {x=0}
    lookAtTarget {boneReference {boneName=pectoralis_02_r_aim}}
    lookAt_Axis {axis=<1 0>}
    lookUp_Axis {axis {x=0, z=1}}
   }
   n34 = "Constraint - scapula_base_l (P)"(ComponentPose=@n33, Alpha=1, ConstraintWeights_0=1) {
    alphaBoolBlend {blendOption=Linear}
    boneToModify {boneName=scapula_base_l}
    constraintSetup [{targetBone {boneName=clavicle_l}}]
    constraintWeights [1]
   }
   n35 = Look At - Bone: scapula_00_l(ComponentPose=@n34, Alpha=1) {
    alphaBoolBlend {blendOption=Linear}
    boneToModify {boneName=scapula_00_l}
    lookAtLocation {x=0}
    lookAtTarget {boneReference {boneName=scapula_01_l_aim}}
    lookAt_Axis {axis=<1 0>}
    lookUp_Axis {axis {x=0, z=1}}
   }
   n36 = Look At - Bone: lattissimus_00_l(ComponentPose=@n35, Alpha=1) {
    alphaBoolBlend {blendOption=Linear}
    boneToModify {boneName=lattissimus_00_l}
    lookAtLocation {x=0}
    lookAtTarget {boneReference {boneName=upperarm_l}}
    lookAt_Axis {axis=<1 0>}
    lookUp_Axis {axis {x=0, z=1}}
   }
   n37 = "Constraint - scapula_base_r (P)"(ComponentPose=@n36, Alpha=1, ConstraintWeights_0=1) {
    alphaBoolBlend {blendOption=Linear}
    boneToModify {boneName=scapula_base_r}
    constraintSetup [{targetBone {boneName=clavicle_r}}]
    constraintWeights [1]
   }
   n38 = Look At - Bone: scapula_00_r(ComponentPose=@n37, Alpha=1) {
    alphaBoolBlend {blendOption=Linear}
    boneToModify {boneName=scapula_00_r}
    lookAtLocation {x=0}
    lookAtTarget {boneReference {boneName=scapula_01_r_aim}}
    lookAt_Axis {axis=<1 0>}
    lookUp_Axis {axis {x=0, z=1}}
   }
   n39 = Look At - Bone: lattissimus_00_r(ComponentPose=@n38, Alpha=1) {
    alphaBoolBlend {blendOption=Linear}
    boneToModify {boneName=lattissimus_00_r}
    lookAtLocation {x=0}
    lookAtTarget {boneReference {boneName=upperarm_r}}
    lookAt_Axis {axis=<1 0>}
    lookUp_Axis {axis {x=0, z=1}}
   }
   n40 = "Transform (Modify) Bone - Bone: bicep_02_l"(ComponentPose=@n39, Alpha=@n29, Translation="-6,-0.000,-0.000", Rotation="0,0,0", Scale="1,1,1") {
    rotationMode=BMM_Additive
    rotationSpace=BCS_BoneSpace
    scaleMode=BMM_Additive
    scaleSpace=BCS_BoneSpace
    translationMode=BMM_Additive
    translationSpace=BCS_BoneSpace
    alphaBoolBlend {blendOption=Linear}
    boneToModify {boneName=bicep_02_l}
    translation {x=-6}
   }
   n41 = "Transform (Modify) Bone - Bone: bicep_01_l"(ComponentPose=@n40, Alpha=@n29, Translation="-0.000,0.000,2.343", Rotation="0,0,0", Scale="1.000,1.500,1.250") {
    rotationMode=BMM_Additive
    rotationSpace=BCS_BoneSpace
    scaleMode=BMM_Additive
    scaleSpace=BCS_BoneSpace
    translationMode=BMM_Additive
    translationSpace=BCS_BoneSpace
    alphaBoolBlend {blendOption=Linear}
    boneToModify {boneName=bicep_01_l}
    scale {y=1.5, z=1.25}
    translation {z=2.343}
   }
   n42 = "Transform (Modify) Bone - Bone: tricep_02_l"(ComponentPose=@n41, Alpha=@n29, Translation="-6,-0.000,-0.000", Rotation="0,0,0", Scale="1,1,1") {
    rotationMode=BMM_Additive
    rotationSpace=BCS_BoneSpace
    scaleMode=BMM_Additive
    scaleSpace=BCS_BoneSpace
    translationMode=BMM_Additive
    translationSpace=BCS_BoneSpace
    alphaBoolBlend {blendOption=Linear}
    boneToModify {boneName=tricep_02_l}
    translation {x=-6}
   }
   n43 = "Transform (Modify) Bone - Bone: tricep_01_l"(ComponentPose=@n42, Alpha=@n29, Translation="-0.000,0.000,2.343", Rotation="0,0,0", Scale="1.000,1.500,1.250") {
    rotationMode=BMM_Additive
    rotationSpace=BCS_BoneSpace
    scaleMode=BMM_Additive
    scaleSpace=BCS_BoneSpace
    translationMode=BMM_Additive
    translationSpace=BCS_BoneSpace
    alphaBoolBlend {blendOption=Linear}
    alphaScaleBiasClamp {scale=0.3}
    boneToModify {boneName=tricep_01_l}
    scale {y=1.5, z=1.25}
    translation {z=2.343}
   }
   n44 = "Transform (Modify) Bone - Bone: bicep_02_r"(ComponentPose=@n43, Alpha=@n28, Translation="-6,-0.000,-0.000", Rotation="0,0,0", Scale="1,1,1") {
    rotationMode=BMM_Additive
    rotationSpace=BCS_BoneSpace
    scaleMode=BMM_Additive
    scaleSpace=BCS_BoneSpace
    translationMode=BMM_Additive
    translationSpace=BCS_BoneSpace
    alphaBoolBlend {blendOption=Linear}
    boneToModify {boneName=bicep_02_r}
    translation {x=-6}
   }
   n45 = "Transform (Modify) Bone - Bone: bicep_01_r"(ComponentPose=@n44, Alpha=@n28, Translation="-0.000,0.000,2.343", Rotation="0,0,0", Scale="1.000,1.500,1.250") {
    rotationMode=BMM_Additive
    rotationSpace=BCS_BoneSpace
    scaleMode=BMM_Additive
    scaleSpace=BCS_BoneSpace
    translationMode=BMM_Additive
    translationSpace=BCS_BoneSpace
    alphaBoolBlend {blendOption=Linear}
    boneToModify {boneName=bicep_01_r}
    scale {y=1.5, z=1.25}
    translation {z=2.343}
   }
   n46 = "Transform (Modify) Bone - Bone: tricep_02_r"(ComponentPose=@n45, Alpha=@n28, Translation="-6,-0.000,-0.000", Rotation="0,0,0", Scale="1,1,1") {
    rotationMode=BMM_Additive
    rotationSpace=BCS_BoneSpace
    scaleMode=BMM_Additive
    scaleSpace=BCS_BoneSpace
    translationMode=BMM_Additive
    translationSpace=BCS_BoneSpace
    alphaBoolBlend {blendOption=Linear}
    alphaScaleBiasClamp {scale=0.5}
    boneToModify {boneName=tricep_02_r}
    translation {x=-6}
   }
   n47 = "Transform (Modify) Bone - Bone: tricep_01_r"(ComponentPose=@n46, Alpha=@n28, Translation="-0.000,0.000,2.343", Rotation="0,0,0", Scale="1.000,1.500,1.250") {
    rotationMode=BMM_Additive
    rotationSpace=BCS_BoneSpace
    scaleMode=BMM_Additive
    scaleSpace=BCS_BoneSpace
    translationMode=BMM_Additive
    translationSpace=BCS_BoneSpace
    alphaBoolBlend {blendOption=Linear}
    alphaScaleBiasClamp {scale=0.3}
    boneToModify {boneName=tricep_01_r}
    scale {y=1.5, z=1.25}
    translation {z=2.343}
   }
   Save cached pose 'anatomyFinal'(Pose=Component To Local(ComponentPose=@n47)) {cacheName=anatomyFinal}
   Use cached pose 'initCache'()
  ]}
 EventGraph {nodes [
   n1 = GetPartSkeletalMeshComponent()
   n2 = TransformFromBoneSpace!(
    execute
    self:Skinned Mesh Component Object Reference=@n1
    BoneName:Name=upperarm_r
    InPosition:Vector="0, 0, 0"
    InRotation:Rotator="0, 0, 0"
   ) -> (then, OutPosition:Vector, OutRotation:Rotator)
   > n3 = TransformFromBoneSpace(self=@n1, BoneName=hand_r, InPosition="0, 0, 0", InRotation="0, 0, 0")
   > Set RarmDistanceAlpha(RarmDistanceAlpha:double=MapRangeClamped(Value="Distance (Vector)"!(V1:Vector=@n2, V2:Vector=@n3) -> (ReturnValue:double), InRangeA=40, InRangeB=54, OutRangeA=0.7, OutRangeB=0)) -> (Output_Get:double)
   Event BlueprintUpdateAnimation()
   > n4 = Sequence!(execute) -> (then_0, then_1, then_2:Exec, then_3:Exec, then_4:Exec, then_5:Exec)
   n5 = Break Fort Anim Input Skydiving External Force(FortAnimInput_SkydivingExternalForce=Get SkydivingExternalForceInput()) -> (
    bUseSkydivingVectorForce=~
    SkydivingVectorForce=~
    HeadToPelvisDirection=~
    FloatingMultiplier=~
    FloatingAdditive=~
    DivingMultiplier=~
    DivingAdditive=~
    ParachutingMultiplier=~
    ParachutingAdditive=~
   )
   > Set windVector(windVector:Vector=SelectVector(A=vector * float(A=@n5.SkydivingVectorForce, B=0.15), B="0, 0,-980", bPickA=Get bIsSkyDiving())) -> (Output_Get:Vector)
   n6 = GetPartSkeletalMeshComponent()
   n7 = TransformFromBoneSpace(execute=@n4.then_3, self=@n6, BoneName=upperarm_l, InPosition="0, 0, 0", InRotation="0, 0, 0")
   > n8 = TransformFromBoneSpace(self=@n6, BoneName=lowerarm_twist_02_l, InPosition="0, 0, 0", InRotation="0, 0, 0")
   > Set LarmDistanceBICEPAlpha(LarmDistanceBICEPAlpha:double=MapRangeClamped(Value="Distance (Vector)"(V1=@n7, V2=@n8), InRangeA=27, InRangeB=34, OutRangeA=1, OutRangeB=0)) -> (Output_Get:double)
   > n9 = TransformFromBoneSpace(self=@n6, BoneName=upperarm_r, InPosition="0, 0, 0", InRotation="0, 0, 0")
   > n10 = TransformFromBoneSpace(self=@n6, BoneName=lowerarm_twist_02_r, InPosition="0, 0, 0", InRotation="0, 0, 0")
   > Set RarmDistanceBICEPAlpha(RarmDistanceBICEPAlpha:double=MapRangeClamped(Value="Distance (Vector)"(V1=@n9, V2=@n10), InRangeA=27, InRangeB=34, OutRangeA=1, OutRangeB=0)) -> (Output_Get:double)
   n11 = GetPartSkeletalMeshComponent()
   n12 = GetAlphaFromReferenceTransformDelta(
    SkeletalMeshComponent=@n11
    BoneName=lowerarm_r
    BaseName=upperarm_r
    DeltaComponentType=QuaternionTwist
    InputRange=(X=0,Y=-90)
    OutputRange=(X=0,Y=0.7)
    TwistAxis="0,0,1"
   )
   Set RflexAlpha(
    execute=@n4.then_2
    RflexAlpha:double=SelectFloat(A=0.25, B=MapRangeClamped(Value=@n12, InRangeA=0, InRangeB=1, OutRangeA=0.2, OutRangeB=0.7), bPickA=Get bIsTargeting())
   ) -> (Output_Get:double)
   n13 = GetAlphaFromReferenceTransformDelta(
    SkeletalMeshComponent=@n11
    BoneName=upperarm_r
    BaseName=clavicle_r
    DeltaComponentType=QuaternionTwist
    InputRange=(X=0,Y=-128)
    TwistAxis="0,0,1"
   )
   > Set RshoulderAlpha(RshoulderAlpha:double=@n13) -> (Output_Get:double)
   > Set RaimPecPos(RaimPecPos:Vector=MakeVector(Y=0, Z=MapRangeClamped(Value=@n13, InRangeA=0, InRangeB=1, OutRangeB=2))) -> (Output_Get:Vector)
   > Set RpecAimPosAlpha(RpecAimPosAlpha:double=@n13) -> (Output_Get:double)
   n14 = GetAlphaFromReferenceTransformDelta(
    SkeletalMeshComponent=@n11
    BoneName=upperarm_l
    BaseName=clavicle_l
    DeltaComponentType=QuaternionTwist
    InputRange=(X=-20,Y=-75)
    TwistAxis="0,0,1"
   )
   > Set LshoulderAlpha(LshoulderAlpha:double=@n14) -> (Output_Get:double)
   > Set LaimPecPos(LaimPecPos:Vector=MakeVector(Y=0, Z=MapRangeClamped(Value=@n14, InRangeA=0, InRangeB=1, OutRangeB=10))) -> (Output_Get:Vector)
   > Set RlegAlpha(RlegAlpha:double=SelectFloat(
     A=1
     B=GetAlphaFromReferenceTransformDelta(SkeletalMeshComponent=@n11, BoneName=thigh_r, BaseName=pelvis, DeltaComponentType=QuaternionTwist, InputRange=(X=0,Y=100), TwistAxis="0,0,1")
     bPickA=Get bIsCrouching()
    )) -> (Output_Get:double)
   > Set LbicepFlex(LbicepFlex:double=GetAlphaFromReferenceTransformDelta(
     SkeletalMeshComponent=@n11
     BoneName=lowerarm_l
     BaseName=upperarm_l
     DeltaComponentType=QuaternionTwist
     InputRange=(X=-20,Y=-70)
     TwistAxis="0,0,1"
    )) -> (Output_Get:double)
   > Set RbicepFlex(RbicepFlex:double=GetAlphaFromReferenceTransformDelta(
     SkeletalMeshComponent=@n11
     BoneName=lowerarm_r
     BaseName=upperarm_r
     DeltaComponentType=QuaternionTwist
     InputRange=(X=-20,Y=-70)
     TwistAxis="0,0,1"
    )) -> (Output_Get:double)
   > n15 = TransformFromBoneSpace(self=@n1, BoneName=upperarm_l, InPosition="0, 0, 0", InRotation="0, 0, 0")
   > n16 = TransformFromBoneSpace(self=@n1, BoneName=lowerarm_twist_02_l, InPosition="0, 0, 0", InRotation="0, 0, 0")
   > Set LarmDistanceBICEPAlpha(LarmDistanceBICEPAlpha=MapRangeClamped(Value="Distance (Vector)"(V1=@n15, V2=@n16), InRangeA=27, InRangeB=34, OutRangeA=1, OutRangeB=0))
   > n17 = TransformFromBoneSpace(self=@n1, BoneName=upperarm_r, InPosition="0, 0, 0", InRotation="0, 0, 0")
   > n18 = TransformFromBoneSpace(self=@n1, BoneName=lowerarm_twist_02_r, InPosition="0, 0, 0", InRotation="0, 0, 0")
   > Set RarmDistanceBICEPAlpha(RarmDistanceBICEPAlpha=MapRangeClamped(Value="Distance (Vector)"(V1=@n17, V2=@n18), InRangeA=27, InRangeB=34, OutRangeA=1, OutRangeB=0))
   Set isSkydiveAlpha(execute=@n4.then_1, isSkydiveAlpha:double="To Float (Boolean)"(InBool=Get bIsSkyDiving())) -> (Output_Get:double)
   n19 = Get PawnSpeed2D()
   n20 = Get bIsCrouching()
   > Set isCrouchingAlpha(isCrouchingAlpha:double=SelectFloat(A=MapRangeClamped(Value=@n19, InRangeA=0, InRangeB=150, OutRangeA=1, OutRangeB=0), B=0, bPickA=@n20)) -> (Output_Get:double)
   n21 = Get CurrentLOD()
   n22 = Get bRigidBodyEnabled()
   Set rigids_off(execute=@n4.then_4, rigids_off:bool=OR Boolean(A=NOT Boolean(A=@n22), B="integer > integer"(A=@n21, B=1) {is=title K2Node_CallFunction})) -> (Output_Get:bool)
   n23 = GetPartSkeletalMeshComponent()
   n24 = GetAlphaFromReferenceTransformDelta(
    SkeletalMeshComponent=@n23
    BoneName=lowerarm_r
    BaseName=upperarm_r
    DeltaComponentType=QuaternionTwist
    InputRange=(X=0,Y=-90)
    OutputRange=(X=0,Y=0.7)
    TwistAxis="0,0,1"
   )
   Set RbicepAlpha(RbicepAlpha:double=@n24) -> (Output_Get:double)
   > Set RflexAlpha(RflexAlpha=MapRangeClamped(Value=@n24, InRangeA=0, InRangeB=1, OutRangeB=0.2))
   n25 = GetAlphaFromReferenceTransformDelta(
    SkeletalMeshComponent=@n23
    BoneName=lowerarm_l
    BaseName=upperarm_l
    DeltaComponentType=QuaternionTwist
    InputRange=(X=0,Y=-90)
    OutputRange=(X=0,Y=0.7)
    TwistAxis="0,0,1"
   )
   > Set LbicepAlpha(LbicepAlpha:double=@n25) -> (Output_Get:double)
   > Set LflexAlpha(LflexAlpha:double=MapRangeClamped(Value=@n25, InRangeA=0, InRangeB=1, OutRangeB=0.2)) -> (Output_Get:double)
   > Set RscaleShoulder(RscaleShoulder:double=GetAlphaFromReferenceTransformDelta(
     SkeletalMeshComponent=@n23
     BoneName=upperarm_r
     BaseName=clavicle_r
     DeltaComponentType=QuaternionTwist
     InputRange=(X=-90,Y=-130)
     TwistAxis="0,0,1"
    )) -> (Output_Get:double)
   > Set LscaleShoulder(LscaleShoulder:double=GetAlphaFromReferenceTransformDelta(
     SkeletalMeshComponent=@n23
     BoneName=upperarm_l
     BaseName=clavicle_l
     DeltaComponentType=QuaternionTwist
     InputRange=(X=-90,Y=-130)
     TwistAxis="0,0,1"
    )) -> (Output_Get:double)
   TryGetPawnOwner()
   "To Float (Boolean)"(InBool=@n20)
   DrawDebugLine!(
    execute
    LineStart:Vector=@n2
    LineEnd:Vector=@n3
    LineColor:Linear Color Structure=(R=0.026,G=1,B=0,A=1)
    Duration:float=0
    Thickness:float=1
    DepthPriority:EDrawDebugSceneDepthPriorityGroup Enum=World
   ) -> (then)
  ]}
]
variables [
 isCrouchingAlpha {type=number}
 isSkydiveAlpha {type=number}
 LaimPecPos {type=Vector}
 LarmDistanceBICEPAlpha {type=number}
 LbicepAlpha {type=number}
 LbicepFlex {type=number}
 LflexAlpha {type=number}
 LscaleShoulder {type=number}
 LshoulderAlpha {type=number}
 RaimPecPos {type=Vector}
 RarmDistanceAlpha {type=number}
 RarmDistanceBICEPAlpha {type=number}
 RbicepAlpha {type=number}
 RbicepFlex {type=number}
 RflexAlpha {type=number}
 rigids_off {type=boolean}
 RlegAlpha {type=number}
 RpecAimPosAlpha {type=number}
 RscaleShoulder {type=number}
 RshoulderAlpha {type=number}
 windVector {type=Vector}
]
