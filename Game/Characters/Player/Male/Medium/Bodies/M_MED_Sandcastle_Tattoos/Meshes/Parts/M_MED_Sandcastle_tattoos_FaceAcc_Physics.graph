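# Physics asset for the face accessory part of the male medium Sandcastle Tattoos player body, set up for dynamic hair.
# Front strands use boxes and long side and back strands use capsules, with head, neck, upper spine and upper-arm capsules as body colliders.
# Swing- and twist-limited constraints chain segments, mostly from tiny MSTR root bodies.
# Angularly free constraints cross-link some neighboring strands and tie the lower-front roots to the head.
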PhysicsAsset /Game/Characters/Player/Male/Medium/Bodies/M_MED_Sandcastle_Tattoos/Meshes/Parts/M_MED_Sandcastle_tattoos_FaceAcc_Physics
bodies {
 dyn_hairfront_01_l [{shapeType=Box, extentX=6, extentY=3, extentZ=1, center=<2.9269 0 -0.0001>}]
 dyn_hairfront_01_l_MSTR [{shapeType=Capsule, length=0.01, radius=0.01}]
 dyn_hairfront_01_r [{shapeType=Box, extentX=6, extentY=2.5, extentZ=1, center=<2.9269 0 -0.0001>}]
 dyn_hairfront_01_r_MSTR [{shapeType=Capsule, length=0.01, radius=0.01}]
 dyn_hairfront_02_l [{shapeType=Box, extentX=6, extentY=3, extentZ=1, center=<2.9269 0 -0.0001>}]
 dyn_hairfront_02_r [{shapeType=Box, extentX=6, extentY=2.5, extentZ=1, center=<2.9269 0 -0.0001>}]
 dyn_hairfront_03_l [{shapeType=Box, extentX=6, extentY=3, extentZ=1, center=<2.9269 0 -0.0001>}]
 dyn_hairfront_03_r [{shapeType=Box, extentX=6, extentY=2.5, extentZ=1, center=<2.9269 0 -0.0001>}]
 dyn_hairfront_bt_01_l [{shapeType=Box, extentX=6, extentY=5, extentZ=1, center=<2.9807 0 0.5931>, rotation=<rot -7.3972 0 0>}]
 dyn_hairfront_bt_01_l_MSTR [{shapeType=Capsule, length=0.01, radius=0.01}]
 dyn_hairfront_bt_01_r [{shapeType=Box, extentX=6, extentY=6.5, extentZ=1, center=<2.9269 0 -0.0001>}]
 dyn_hairfront_bt_01_r_MSTR [{shapeType=Capsule, length=0.01, radius=0.01}]
 dyn_hairfront_bt_02_l [{shapeType=Box, extentX=6, extentY=5.5, extentZ=1, center=<2.9269 0 -0.0001>}]
 dyn_hairfront_bt_02_r [{shapeType=Box, extentX=6, extentY=6.5, extentZ=1, center=<2.9269 0 -0.0001>}]
 dyn_hairfront_bt_03_l [{shapeType=Box, extentX=6, extentY=4.5, extentZ=1, center=<2.9269 0 0>, rotation=<rot 0 0 2.6673>}]
 dyn_hairfront_bt_03_r [{shapeType=Box, extentX=6, extentY=6.5, extentZ=1, center=<2.9269 0 -0.0001>}]
 dyn_hairfront_bt_04_l [{shapeType=Box, extentX=6, extentY=4, extentZ=1, center=<2.9269 0.1173 0>, rotation=<rot 0 0 9.4907>}]
 dyn_hairfront_bt_04_r [{shapeType=Box, extentX=6, extentY=6.5, extentZ=1, center=<2.9269 0 -0.0001>}]
 dyn_hairfront_bt_05_l [{shapeType=Box, extentX=5, extentY=3, extentZ=1, center=<2.2848 0 0>}]
 dyn_hairfront_bt_05_r [{shapeType=Box, extentX=6, extentY=6.5, extentZ=1, center=<2.9269 0 -0.0001>}]
 dyn_hairfront_bt_06_l [{shapeType=Box, extentX=5, extentY=3, extentZ=1, center=<2.2848 0 0>}]
 dyn_hairfront_tp_01_l [{shapeType=Box, extentX=6, extentY=2, extentZ=1, center=<2.9269 0 -0.0001>}]
 dyn_hairfront_tp_01_l_MSTR [{shapeType=Capsule, length=0.01, radius=0.01}]
 dyn_hairfront_tp_01_r [{shapeType=Box, extentX=6, extentY=2, extentZ=1, center=<2.9269 0 -0.0001>}]
 dyn_hairfront_tp_01_r_MSTR [{shapeType=Capsule, length=0.01, radius=0.01}]
 dyn_hairfront_tp_02_l [{shapeType=Box, extentX=6, extentY=2, extentZ=1, center=<2.9269 0 -0.0001>}]
 dyn_hairfront_tp_02_r [{shapeType=Box, extentX=6, extentY=2, extentZ=1, center=<2.9269 0 -0.0001>}]
 dyn_hairlong_01_l_MSTR [{shapeType=Capsule, length=0.01, radius=0.01}]
 dyn_hairlong_01_r_MSTR [{shapeType=Capsule, length=0.01, radius=0.01}]
 dyn_hairlong_02_l [{shapeType=Capsule, length=5.0251, radius=2, center=<2.4216 0 0>, rotation=<rot -90 -179.9802 179.9803>}]
 dyn_hairlong_02_r [{shapeType=Capsule, length=5.0251, radius=2.3907, center=<3.1927 0 0>, rotation=<rot -90 0.0197 -0.0197>}]
 dyn_hairlong_03_l [{shapeType=Capsule, length=5.0251, radius=2, center=<2.6874 0 0.3461>, rotation=<rot -90 0.0264 -0.0263>}]
 dyn_hairlong_03_r [{shapeType=Capsule, length=5.0251, radius=2.3907, center=<3.1927 0 0>, rotation=<rot -90 0.0197 -0.0197>}]
 dyn_hairlong_04_l [{shapeType=Capsule, length=4.6, radius=1.9, center=<2.1679 0.0001 0>, rotation=<rot -90 -0.0149 0.0149>}]
 dyn_hairlong_04_r [{shapeType=Capsule, length=5.0251, radius=2.3907, center=<3.1927 0 0>, rotation=<rot -90 0.0197 -0.0197>}]
 dyn_hairlong_bk_01 [{shapeType=Capsule, length=5.5251, radius=1.8907, center=<3.1927 0 0>, rotation=<rot -90 0.0197 -0.0197>}]
 dyn_hairlong_bk_01_MSTR [{shapeType=Capsule, length=0.01, radius=0.01}]
 dyn_hairlong_bk_01_l [{shapeType=Capsule, length=5.5251, radius=1.8907, center=<3.1927 0 0>, rotation=<rot -90 0.0197 -0.0197>}]
 dyn_hairlong_bk_01_l_MSTR [{shapeType=Capsule, length=0.01, radius=0.01}]
 dyn_hairlong_bk_01_r [{shapeType=Capsule, length=5.5251, radius=1.8907, center=<3.1927 0 0>, rotation=<rot -90 0.0197 -0.0197>}]
 dyn_hairlong_bk_01_r_MSTR [{shapeType=Capsule, length=0.01, radius=0.01}]
 dyn_hairlong_bk_02 [{shapeType=Capsule, length=5.5251, radius=1.8907, center=<3.1927 0 0>, rotation=<rot -90 0.0197 -0.0197>}]
 dyn_hairlong_bk_02_l [{shapeType=Capsule, length=5.5251, radius=1.8907, center=<3.1927 0 0>, rotation=<rot -90 0.0197 -0.0197>}]
 dyn_hairlong_bk_02_r [{shapeType=Capsule, length=5.5251, radius=1.8907, center=<3.1927 0 0>, rotation=<rot -90 0.0197 -0.0197>}]
 dyn_hairlong_bk_03 [{shapeType=Capsule, length=5.5251, radius=1.8907, center=<3.1927 0 0>, rotation=<rot -90 0.0197 -0.0197>}]
 dyn_hairlong_bk_03_l [{shapeType=Capsule, length=5.5251, radius=1.8907, center=<3.1927 0 0>, rotation=<rot -90 0.0197 -0.0197>}]
 dyn_hairlong_bk_03_r [{shapeType=Capsule, length=5.5251, radius=1.8907, center=<3.1927 0 0>, rotation=<rot -90 0.0197 -0.0197>}]
 dyn_hairlong_bk_04 [{shapeType=Capsule, length=5.5251, radius=1.8907, center=<3.1927 0 0>, rotation=<rot -90 0.0197 -0.0197>}]
 dyn_hairlong_bk_04_l [{shapeType=Capsule, length=5.5251, radius=1.8907, center=<3.1927 0 0>, rotation=<rot -90 0.0197 -0.0197>}]
 dyn_hairlong_bk_04_r [{shapeType=Capsule, length=5.5251, radius=1.8907, center=<3.1927 0 0>, rotation=<rot -90 0.0197 -0.0197>}]
 dyn_hairlong_bk_05 [{shapeType=Capsule, length=5.5251, radius=1.8907, center=<3.1927 0 0>, rotation=<rot -90 0.0197 -0.0197>}]
 dyn_hairlong_bk_05_l [{shapeType=Capsule, length=5.5251, radius=1.8907, center=<3.1927 0 0>, rotation=<rot -90 0.0197 -0.0197>}]
 dyn_hairlong_bk_05_r [{shapeType=Capsule, length=5.5251, radius=1.8907, center=<3.1927 0 0>, rotation=<rot -90 0.0197 -0.0197>}]
 head [{shapeType=Capsule, length=6.75, radius=7.25, center=<3.3927 1.1113 0>, rotation=<rot 90 -0.0099 -0.0099>}]
 neck_01 [{shapeType=Capsule, length=8, radius=6, center=<3.3927 1.1113 0>, rotation=<rot 90 -0.0099 -0.0099>}]
 spine_04_COL [{shapeType=Capsule, length=15, radius=15, center=<0 -4.6196 0.0001>}]
 spine_05_COL [{shapeType=Capsule, length=18.25, radius=11.75, center=<0.4146 -2.258 0>}]
 upperarm_l [{shapeType=Capsule, length=19.25, radius=8.005, center=<8.8498 -1.3674 -2.6872>, rotation=<rot 0 -89.9995 100.0001>}]
 upperarm_r [{shapeType=Capsule, length=19.25, radius=8.005, center=<-8.8498 1.3674 2.6872>, rotation=<rot 0 -89.9995 -79.9998>}]
}
constraints [
 {bone1Name=dyn_hairlong_03_l, bone2Name=dyn_hairlong_02_l, swing1LimitDegrees=45, swing2LimitDegrees=5, twistLimitDegrees=5}
 {bone1Name=dyn_hairlong_04_l, bone2Name=dyn_hairlong_03_l, swing1LimitDegrees=45, swing2LimitDegrees=5, twistLimitDegrees=5}
 {bone1Name=dyn_hairlong_bk_01_l, bone2Name=dyn_hairlong_bk_01_l_MSTR, swing1LimitDegrees=5, swing2LimitDegrees=15, twistLimitDegrees=5}
 {bone1Name=dyn_hairlong_bk_02_l, bone2Name=dyn_hairlong_bk_01_l, swing1LimitDegrees=5, swing2LimitDegrees=45, twistLimitDegrees=5}
 {bone1Name=dyn_hairlong_bk_03_l, bone2Name=dyn_hairlong_bk_02_l, swing1LimitDegrees=5, swing2LimitDegrees=45, twistLimitDegrees=5}
 {bone1Name=dyn_hairlong_bk_04_l, bone2Name=dyn_hairlong_bk_03_l, swing1LimitDegrees=5, swing2LimitDegrees=45, twistLimitDegrees=5}
 {bone1Name=dyn_hairlong_bk_05_l, bone2Name=dyn_hairlong_bk_04_l, swing1LimitDegrees=5, swing2LimitDegrees=45, twistLimitDegrees=5}
 {bone1Name=dyn_hairlong_bk_01, bone2Name=dyn_hairlong_bk_01_MSTR, swing1LimitDegrees=5, swing2LimitDegrees=45, twistLimitDegrees=5}
 {bone1Name=dyn_hairlong_bk_02, bone2Name=dyn_hairlong_bk_01, swing1LimitDegrees=5, swing2LimitDegrees=45, twistLimitDegrees=5}
 {bone1Name=dyn_hairlong_bk_03, bone2Name=dyn_hairlong_bk_02, swing1LimitDegrees=5, swing2LimitDegrees=45, twistLimitDegrees=5}
 {bone1Name=dyn_hairlong_bk_04, bone2Name=dyn_hairlong_bk_03, swing1LimitDegrees=5, swing2LimitDegrees=45, twistLimitDegrees=5}
 {bone1Name=dyn_hairlong_bk_05, bone2Name=dyn_hairlong_bk_04, swing1LimitDegrees=5, swing2LimitDegrees=45, twistLimitDegrees=5}
 {bone1Name=dyn_hairfront_tp_01_l, bone2Name=dyn_hairfront_tp_01_l_MSTR, swing1LimitDegrees=20, swing2LimitDegrees=20, twistLimitDegrees=5}
 {bone1Name=dyn_hairfront_tp_02_l, bone2Name=dyn_hairfront_tp_01_l, swing1LimitDegrees=20, swing2LimitDegrees=20, twistLimitDegrees=5}
 {bone1Name=dyn_hairfront_01_l, bone2Name=dyn_hairfront_01_l_MSTR, swing1LimitDegrees=10, swing2LimitDegrees=10, twistLimitDegrees=5}
 {bone1Name=dyn_hairfront_02_l, bone2Name=dyn_hairfront_01_l, swing1LimitDegrees=20, swing2LimitDegrees=20, twistLimitDegrees=5}
 {bone1Name=dyn_hairfront_03_l, bone2Name=dyn_hairfront_02_l, swing1LimitDegrees=5, swing2LimitDegrees=25, twistLimitDegrees=5}
 {bone1Name=dyn_hairlong_bk_01_r, bone2Name=dyn_hairlong_bk_01_r_MSTR, swing1LimitDegrees=5, swing2LimitDegrees=45, twistLimitDegrees=5}
 {bone1Name=dyn_hairlong_bk_02_r, bone2Name=dyn_hairlong_bk_01_r, swing1LimitDegrees=5, swing2LimitDegrees=45, twistLimitDegrees=5}
 {bone1Name=dyn_hairlong_bk_03_r, bone2Name=dyn_hairlong_bk_02_r, swing1LimitDegrees=5, swing2LimitDegrees=45, twistLimitDegrees=5}
 {bone1Name=dyn_hairlong_bk_04_r, bone2Name=dyn_hairlong_bk_03_r, swing1LimitDegrees=5, swing2LimitDegrees=45, twistLimitDegrees=5}
 {bone1Name=dyn_hairlong_bk_05_r, bone2Name=dyn_hairlong_bk_04_r, swing1LimitDegrees=5, swing2LimitDegrees=45, twistLimitDegrees=5}
 {bone1Name=dyn_hairlong_03_r, bone2Name=dyn_hairlong_02_r, swing1LimitDegrees=45, swing2LimitDegrees=5, twistLimitDegrees=5}
 {bone1Name=dyn_hairlong_04_r, bone2Name=dyn_hairlong_03_r, swing1LimitDegrees=45, swing2LimitDegrees=5, twistLimitDegrees=5}
 {bone1Name=dyn_hairfront_01_r, bone2Name=dyn_hairfront_01_r_MSTR, swing1LimitDegrees=10, swing2LimitDegrees=10, twistLimitDegrees=5}
 {bone1Name=dyn_hairfront_02_r, bone2Name=dyn_hairfront_01_r, swing1LimitDegrees=10, swing2LimitDegrees=10, twistLimitDegrees=5}
 {bone1Name=dyn_hairfront_03_r, bone2Name=dyn_hairfront_02_r, swing1LimitDegrees=10, swing2LimitDegrees=10, twistLimitDegrees=5}
 {bone1Name=dyn_hairfront_tp_01_r, bone2Name=dyn_hairfront_tp_01_r_MSTR, swing1LimitDegrees=20, swing2LimitDegrees=20, twistLimitDegrees=5}
 {bone1Name=dyn_hairfront_tp_02_r, bone2Name=dyn_hairfront_tp_01_r, swing1LimitDegrees=20, swing2LimitDegrees=20, twistLimitDegrees=5}
 {bone1Name=dyn_hairfront_bt_01_r, bone2Name=dyn_hairfront_bt_01_r_MSTR, swing1LimitDegrees=15, swing2LimitDegrees=15, twistLimitDegrees=5}
 {bone1Name=dyn_hairfront_bt_02_r, bone2Name=dyn_hairfront_bt_01_r, swing1LimitDegrees=15, swing2LimitDegrees=15, twistLimitDegrees=5}
 {bone1Name=dyn_hairfront_bt_03_r, bone2Name=dyn_hairfront_bt_02_r, swing1LimitDegrees=15, swing2LimitDegrees=15, twistLimitDegrees=5}
 {bone1Name=dyn_hairfront_bt_04_r, bone2Name=dyn_hairfront_bt_03_r, swing1LimitDegrees=45, swing2LimitDegrees=45, twistLimitDegrees=5}
 {bone1Name=dyn_hairfront_bt_05_r, bone2Name=dyn_hairfront_bt_04_r, swing1LimitDegrees=45, swing2LimitDegrees=45, twistLimitDegrees=5}
 {bone1Name=dyn_hairfront_bt_01_l, bone2Name=dyn_hairfront_bt_01_l_MSTR, swing1LimitDegrees=15, swing2LimitDegrees=15, twistLimitDegrees=5}
 {bone1Name=dyn_hairfront_bt_02_l, bone2Name=dyn_hairfront_bt_01_l, swing1LimitDegrees=15, swing2LimitDegrees=15, twistLimitDegrees=5}
 {bone1Name=dyn_hairfront_bt_03_l, bone2Name=dyn_hairfront_bt_02_l, swing1LimitDegrees=15, swing2LimitDegrees=15, twistLimitDegrees=5}
 {bone1Name=dyn_hairfront_bt_04_l, bone2Name=dyn_hairfront_bt_03_l, swing1LimitDegrees=45, swing2LimitDegrees=45, twistLimitDegrees=5}
 {bone1Name=dyn_hairfront_bt_05_l, bone2Name=dyn_hairfront_bt_04_l, swing1LimitDegrees=45, swing2LimitDegrees=45, twistLimitDegrees=5}
 {bone1Name=dyn_hairfront_bt_06_l, bone2Name=dyn_hairfront_bt_04_l, swing1LimitDegrees=45, swing2LimitDegrees=45, twistLimitDegrees=5}
 {
  bone1Name=dyn_hairfront_bt_01_r_MSTR
  bone2Name=head
  swing1LimitDegrees=45
  swing1Motion=Free
  swing2LimitDegrees=45
  swing2Motion=Free
  twistLimitDegrees=45
  twistMotion=Free
 }
 {
  bone1Name=dyn_hairfront_bt_01_l_MSTR
  bone2Name=head
  swing1LimitDegrees=45
  swing1Motion=Free
  swing2LimitDegrees=45
  swing2Motion=Free
  twistLimitDegrees=45
  twistMotion=Free
 }
 {
  bone1Name=dyn_hairlong_bk_05_r
  bone2Name=dyn_hairlong_bk_05
  swing1LimitDegrees=45
  swing1Motion=Free
  swing2LimitDegrees=45
  swing2Motion=Free
  twistLimitDegrees=45
  twistMotion=Free
 }
 {
  bone1Name=dyn_hairlong_bk_05_l
  bone2Name=dyn_hairlong_bk_05
  swing1LimitDegrees=45
  swing1Motion=Free
  swing2LimitDegrees=45
  swing2Motion=Free
  twistLimitDegrees=45
  twistMotion=Free
 }
 {
  bone1Name=dyn_hairlong_bk_04_r
  bone2Name=dyn_hairlong_04_r
  swing1LimitDegrees=45
  swing1Motion=Free
  swing2LimitDegrees=45
  swing2Motion=Free
  twistLimitDegrees=45
  twistMotion=Free
 }
 {
  bone1Name=dyn_hairfront_bt_02_l
  bone2Name=dyn_hairlong_03_l
  swing1LimitDegrees=45
  swing1Motion=Free
  swing2LimitDegrees=45
  swing2Motion=Free
  twistLimitDegrees=45
  twistMotion=Free
 }
 {
  bone1Name=dyn_hairfront_bt_02_r
  bone2Name=dyn_hairlong_03_r
  swing1LimitDegrees=45
  swing1Motion=Free
  swing2LimitDegrees=45
  swing2Motion=Free
  twistLimitDegrees=45
  twistMotion=Free
 }
 {
  bone1Name=dyn_hairlong_03_r
  bone2Name=dyn_hairlong_bk_03_r
  swing1LimitDegrees=45
  swing1Motion=Free
  swing2LimitDegrees=45
  swing2Motion=Free
  twistLimitDegrees=45
  twistMotion=Free
 }
 {
  bone1Name=dyn_hairlong_bk_04_r
  bone2Name=dyn_hairlong_bk_04
  swing1LimitDegrees=45
  swing1Motion=Free
  swing2LimitDegrees=45
  swing2Motion=Free
  twistLimitDegrees=45
  twistMotion=Free
 }
 {
  bone1Name=dyn_hairlong_bk_04_l
  bone2Name=dyn_hairlong_bk_04
  swing1LimitDegrees=45
  swing1Motion=Free
  swing2LimitDegrees=45
  swing2Motion=Free
  twistLimitDegrees=45
  twistMotion=Free
 }
 {
  bone1Name=dyn_hairfront_bt_03_r
  bone2Name=dyn_hairlong_04_r
  swing1LimitDegrees=45
  swing1Motion=Free
  swing2LimitDegrees=45
  swing2Motion=Free
  twistLimitDegrees=45
  twistMotion=Free
 }
]
